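; Vault lock that unlocks when a scanned player carries the required medallion, removing itself if disabled by its enabling row.
; Entering the scan volume starts an animated ray, sound and timer, which checks every cached pawn's inventory.
; Success sets the lock state and sends quest credit; failure applies the fail effect.
; Both outcomes update the holo widget and flash a dynamic emissive material, while custom primitive data styles the ray.
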(event EventBeginPlay
  (bind _lock (Variables|Default|GetLock))
  (|CallParentFunction)
  (Class|FortAsyncActionCurrentPlaylistReady|PlaylistDataReadyAsync
    (:OnReady
      (if (ScalableFloat|GetValueasBool (Variables|Default|GetRow_Enabled))
        ;[1 1 1] Initialize widget and scanner ray prim data - not sure why this doesn't work in construction script, so we're doing both to be able to previw in viewport and make it work in-game
        (CallFunction|InitializeScannerRay (Variables|Scanner|GetScannerBaseWidth) (Variables|Scanner|GetScannerEndWidth) (Variables|Scanner|GetScannerColorDefault) (Variables|Scanner|GetScannerAngle) (Variables|Scanner|GetScannerAnimRate))
        (Rendering|SetVisibility (Variables|Default|GetScannerRay) false false)
        (CallFunction|CacheHoloWidget)
        (CallFunction|UpdateWidgetItemIcon)
        ;[1 1 1] Create and cache dynamic material for the lock to animate later
        (bind _returnvalue (Rendering|Material|CreateDynamicMaterialInstance _lock 0 0 "None" :!class /Script/Engine.PrimitiveComponent))
        (bind _output_get (Variables|Scanner|SetLockDynamicMaterial _returnvalue))
        (Rendering|Material|SetMaterial _lock 0 _output_get)
        (else
          (Actor|DestroyActor))))))

(event OnComponentBeginOverlap(ScanVolume)
  (params (OverlappedComponent :type Primitive%20Component%20Object%20Reference) (OtherActor :type Actor%20Object%20Reference) (OtherComp :type Primitive%20Component%20Object%20Reference) (OtherBodyIndex :type Integer) (bFromSweep :type Boolean) (SweepResult :type Hit%20Result%20Structure%20%28by%20ref%29))
  ;[1 1 1] Check overlaps with scanner volume to start scan
  (if (Utilities|Enum|NotEqual(Enum) (Variables|LockParent|GetLockState) "NewEnumerator2")
    (bind _asfort_player_pawn (Utilities|Casting|CastToFortPlayerPawn OtherActor))
    ;;[1 1 1] Cache all pawns in volume - scan will check them all once it reaches that point in the sequence
    (Utilities|Array|AddUnique (Variables|Scanner|GetPawnsInScanVolume) _asfort_player_pawn)
    (if (not (Variables|Scanner|GetIsScanning))
      (CallFunction|StartScanning))))

(event OnComponentEndOverlap(ScanVolume)
  (params (OverlappedComponent :type Primitive%20Component%20Object%20Reference) (OtherActor :type Actor%20Object%20Reference) (OtherComp :type Primitive%20Component%20Object%20Reference) (OtherBodyIndex :type Integer))
  ;[1 1 1] Check overlaps with scanner volume to start scan
  (bind _asfort_player_pawn (Utilities|Casting|CastToFortPlayerPawn OtherActor))
  (Utilities|Array|RemoveItem (Variables|Scanner|GetPawnsInScanVolume) _asfort_player_pawn)
  (CallFunction|RemoveScanningCuefromPawn _asfort_player_pawn))

(event Custom|StartScanning
  ;[1 1 1] Scanning Stages
  ;;[1 1 1] Start scanning
  (Variables|Scanner|SetIsScanning true)
  (Class|WLockTDMHolo|EventScanning (Variables|Scanner|GetHoloWidget))
  (CallFunction|SetScannerColor (Variables|Default|GetScannerRay) (Variables|Scanner|GetScannerColorDefault))
  (Utilities|Time|SetTimerbyEvent (EventDispatchers|CreateEvent :!event Custom|CheckScanStatus) (Variables|Scanner|GetScanDuration))
  (Rendering|SetVisibility (Variables|Default|GetScannerRay) true false)
  (Audio|Components|Audio|FadeIn (Variables|Default|GetScanSound) 0.1 1.0 0.0 "Linear" :!class /Script/Engine.AudioComponent)
  (Audio|PlaySoundatLocation (Variables|Scanner|GetScanStartupSound) (Transformation|GetActorLocation))
  (CallFunction|AddScanningCuetoAllPawns)
  (CallFunction|SetScannerRayAnimRate 0.0)
  (bind (_direction _angle) (Timeline|TL_ScannerAngle PlayFromStart)
    (:Update
      (CallFunction|SetScannerRayAngle _angle (Variables|Default|GetScannerRay)))
    (:Finished
      (CallFunction|SetScannerRayAnimRate (Variables|Scanner|GetScannerAnimRate)))))

(event Custom|CheckScanStatus
  ;[1 1 1] Scanning Stages
  ;;[1 1 1] Display outcome of scan
  (bind _success (CallFunction|CheckAllPawns)))

(event Custom|StopScanning
  ;[1 1 1] Scanning Stages
  ;;[1 1 1] Stop Scanning
  (if (Variables|Scanner|GetIsScanning)
    (Variables|Scanner|SetIsScanning false)
    (Rendering|SetVisibility (Variables|Default|GetScannerRay) false false)
    (Audio|Components|Audio|FadeOut (Variables|Default|GetScanSound) 0.3 0.0 "Linear" :!class /Script/Engine.AudioComponent)
    (Audio|PlaySoundatLocation (Variables|Scanner|GetScanStopSound) (Transformation|GetActorLocation))
    (CallFunction|RemoveScanningCuefromAllPawns)))

(event Custom|PlayEmisFlash
  ;[1 1 1] Generic emissive flash - used for both fail and unlock
  (bind (_direction _emissivemult) (Timeline|TL_EmisFlash PlayFromStart)
    (:Update
      (Rendering|Material|SetVectorParameterValue (Variables|Scanner|GetLockDynamicMaterial) "EmissiveMultiplier" (Math|Conversions|ToLinearColor(Float) _emissivemult) :!class /Script/Engine.MaterialInstanceDynamic))))

(event EventApplyFailCheckEffect
  ;[1 1 1] Failed Feedback
  (|CallParentFunction)
  (Class|WLockTDMHolo|EventFail (Variables|Scanner|GetHoloWidget) "")
  (CallFunction|PlaySoundonMe (Variables|Default|GetFailedSound))
  (CallFunction|StopScanning)
  (Variables|Scanner|GetLockDynamicMaterial
    (:then
      (CallFunction|PlayEmisFlash))
    (:else)))

(event EventOnUnlocked
  (bind _lock (Variables|Default|GetLock))
  ;[1 1 1] Unlocked Feedback
  (|CallParentFunction)
  (CallFunction|PlaySoundonMe (Variables|Default|GetSuccessSound))
  (Utilities|FlowControl|Delay (Variables|Default|GetSuccessFlashDelay))
  (CallFunction|StopScanning)
  (Class|WLockTDMHolo|Unlock (Variables|Scanner|GetHoloWidget))
  (bind _returnvalue (Rendering|Material|CreateDynamicMaterialInstance _lock 0 (Variables|Scanner|GetSuccessMaterial) "None" :!class /Script/Engine.PrimitiveComponent))
  (bind _output_get (Variables|Scanner|SetLockDynamicMaterial _returnvalue))
  (Rendering|Material|SetMaterial _lock 0 _output_get)
  (CallFunction|PlayEmisFlash))

(fn AddScanningCueToAllPawns
  (Networking|SwitchHasAuthority
    (:Authority
      (for _array_element (Variables|Scanner|GetPawnsInScanVolume)
        (if (GameplayTags|HasMatchingGameplayTag _array_element (Variables|Default|GetPawnScannedCue))
          _)))))

(fn CacheHoloWidget
  (bind _asw_lock_tdm_holo (Utilities|Casting|CastToW_Lock_TDM_Holo (UserInterface|GetWidget (Variables|Default|GetLockScreenWidget))))
  (Variables|Scanner|SetHoloWidget _asw_lock_tdm_holo))

(fn CheckAllPawns
  (bind _array_element (Utilities|Array|ForEachLoopwithBreak (Variables|Scanner|GetPawnsInScanVolume))
    (:LoopBody
      (bind _checkpassed (CallFunction|LockCheck _array_element))
      (if _checkpassed
        (CallFunction|SetLock "NewEnumerator2")
        (LocalVariables|SetbAnyPawnSucceeded true)
        (CallFunction|GiveQuestCredit _array_element)))
    (:Completed
      _))
  (if (not (LocalVariables|GetbAnyPawnSucceeded))
    (CallFunction|ApplyFailCheckEffect))
  (return (LocalVariables|GetbAnyPawnSucceeded)))

(fn GiveQuestCredit
  (params (Actor :type Actor%20Object%20Reference))
  (bind _returnvalue (Fortnite|GetFortPlayerControllerfromActor Actor))
  (bind _attachedlockactor (Variables|LockParent|GetAttachedLockActor))
  (bind _returnvalue_2 (GameplayTags|GetOwnedGameplayTags _attachedlockactor :!class /Script/GameplayTags.GameplayTagAssetInterface))
  (Utilities|IsValid _returnvalue
    (:Is%20Valid
      (Quest|SendComplexCustomStatEvent (Quests|GetQuestManager _returnvalue "Invalid") 0 "(GameplayTags=((TagName=\"Athena.Quests.UnlockFaction.Chest\")))" _returnvalue_2 1)
      (bind _result (CustomVerb|AddTagsforSubject (CustomVerb|MakeCustomVerbMessage _returnvalue "(TagName=\"CustomVerb.KeyUnlocked\")") "UnlockedActor" _returnvalue_2))
      (CustomVerb|BroadcastCustomVerbMessage _result))))

(fn InitializeScannerRay
  (params (BaseWidth :type Float%20%28single-precision%29) (EndWidth :type Float%20%28single-precision%29) (Color :type Linear%20Color%20Structure) (Angle :type Float%20%28single-precision%29) (ScanRotAnim :type Float%20%28single-precision%29))
  (bind _scannerray (Variables|Default|GetScannerRay))
  (Rendering|Material|SetCustomPrimitiveDataFloat _scannerray 2 BaseWidth) ; WidthBase
  (Rendering|Material|SetCustomPrimitiveDataFloat _scannerray 1 EndWidth) ; WidthEnd
  (Rendering|Material|SetCustomPrimitiveDataFloat _scannerray 12 1.0) ; SecondaryMovementSpeed
  (Rendering|Material|SetCustomPrimitiveDataFloat _scannerray 13 0.5) ; SecondaryFanOpacity
  (CallFunction|SetScannerRayAngle Angle _scannerray)
  (Rendering|Material|SetCustomPrimitiveDataFloat _scannerray 3 (Variables|Scanner|GetScannerFalloffHardness)) ; FalloffHardness
  (CallFunction|SetScannerColor _scannerray Color)
  (CallFunction|SetScannerRayAnimRate ScanRotAnim))

(fn LockCheck
  (params (InteractingPawn :type Fort%20Pawn%20Object%20Reference) (InteractionBeingAttempted :type EInteractionBeingAttempted%20Enum :default "FirstInteraction"))
  (return (> (Inventory|GetItemQuantityOnPlayer (Fortnite|GetFortPlayerControllerfromActor InteractingPawn) (Variables|Default|GetRequiredMedallion)) 0)))

(fn PlaySoundOnMe
  (params (Sound :type Sound%20Base%20Object%20Reference))
  (Audio|PlaySoundatLocation Sound (Transformation|GetActorLocation)))

(fn RemoveScanningCueFromAllPawns
  (for _array_element (Variables|Scanner|GetPawnsInScanVolume)
    (CallFunction|RemoveScanningCuefromPawn _array_element)))

(fn RemoveScanningCueFromPawn
  (params (TargetActor :type Actor%20Object%20Reference))
  (GameplayCue|RemoveGameplayCue TargetActor (Variables|Default|GetPawnScannedCue)))

(fn SetScannerColor
  (params (ScannerRayStaticMesh :type Static%20Mesh%20Component%20Object%20Reference) (InScannerColor :type Linear%20Color%20Structure))
  (bind (_r _g _b _a) (Utilities|Struct|BreakLinearColor InScannerColor))
  (bind (_r_2 _g_2 _b_2 _a_2) (Utilities|Struct|BreakLinearColor InScannerColor))
  (Rendering|Material|SetCustomPrimitiveDataVector4 ScannerRayStaticMesh 4 (Math|Vector4|MakeVector4 _r_2 _g_2 _b_2 _a_2)) ; Color
  (Rendering|Material|SetCustomPrimitiveDataVector4 0 8 (Math|Vector4|MakeVector4 _r _g _b _a)) ; Color
)

(fn SetScannerRayAngle
  (params (Angle :type Float%20%28single-precision%29) (Component :type Primitive%20Component%20Object%20Reference))
  (Rendering|Material|SetCustomPrimitiveDataFloat Component 0 Angle) ; Angle
)

(fn SetScannerRayAnimRate
  (params (Rate :type Float%20%28single-precision%29))
  (Rendering|Material|SetScalarParameterValueonMaterials (Variables|Default|GetScannerRay) "ScanRotAnim" Rate))

(fn UpdateWidgetItemIcon
  (Class|WLockTDMHolo|UpdateRequiredItem (Variables|Scanner|GetHoloWidget) (Variables|Default|GetRequiredMedallion)))

(fn ConstructionScript
  (|CallParentFunction)
  (Utilities|FlowControl|Sequence
    (:then_0
      (CallFunction|InitializeScannerRay (Variables|Scanner|GetScannerBaseWidth) (Variables|Scanner|GetScannerEndWidth) (Variables|Scanner|GetScannerColorDefault) (Variables|Scanner|GetScannerAngle) (Variables|Scanner|GetScannerAnimRate))
      (bind _requiredmedallion (Variables|Default|GetRequiredMedallion)
        (:then
          (CallFunction|UpdateWidgetItemIcon)
          (Rendering|Components|TextRender|SetText (Variables|Default|GetTextRender) (ItemDefinition|GetDisplayName _requiredmedallion)))
        (:else)))))
