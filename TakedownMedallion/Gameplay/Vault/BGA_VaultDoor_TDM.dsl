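; Vault door Blueprint for the Takedown Medallion vault that announces when the vault opens.
; On vault open it runs the parent logic, then gets the FortPlayspace event router and, if valid, broadcasts its opened event struct.
; Per the graph's note, the broadcast lets WeaponModStation mini icons become visible on the map.
; The other events are empty or disabled stubs, and the construction script only calls the parent.
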(event EventVaultOpen
  ;[0 0.159 1] For WeaponModStations mini icon visibility, broadcast vault opened to show the map icon
  (|CallParentFunction)
  (bind _returnvalue (GameplayEventRouter|GetEventRouter "/Script/FortniteGame.FortPlayspace" self))
  (Utilities|IsValid _returnvalue
    (:Is%20Valid
      (GameplayEventRouter|BroadcastEvent _returnvalue (Variables|Default|GetOpenedEventStruct) self))))

(event EventBeginPlay
  (|CallParentFunction) ; This node is disabled and will not be called.\nDrag off pins to build functionality.
)

(event Collision|EventActorBeginOverlap
  (params (OtherActor :type Actor%20Object%20Reference)))

(event EventTick
  (params (DeltaSeconds :type Float%20%28single-precision%29)))

(fn ConstructionScript
  (|CallParentFunction))
